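# Master material for winged insect Niagara mesh particles, using bee textures by default.
# Diffuse color and alpha are multiplied by particle color, alongside a normal map and a SubSurfColor parameter.
# A time-driven sine, masked by vertex red and scaled along world vertex normals, produces procedural wing flapping.
# Dynamic parameters let particles set the flap's time offset, range and speed.
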Material /Game/Effects/Niagara/MasterMaterials/Insects/M_Winged_Insects
graph {
 properties {
  bUsedWithNiagaraMeshParticles=true
  pixelDepthOffsetMode=PDOM_Legacy
  previewMesh=/Game/Athena/Items/Consumables/HoneyPot/Meshes/SM_Bee_Honeypot
  usedShadingModels=Default Lit
 }
 nodes [
  n1 = TextureSampleParameter2D() {parameterName=DiffuseAndAlpha, texture=/Game/Effects/Fort_Effects/Textures/Bugs/Bees/T_Bee_Color}
  n2 = ParticleColor()
  Multiply(A=@n1.RGB, B=@n2.RGB)
  n3 = Time()
  n4 = DynamicParameter() {defaultValue=<rgba 0 8 1 1>, paramNames [FlapRandomOffset, ZFlappingRange, FlapSpeed, Param4]}
  n5 = Sine(None=Multiply(A=Add(A=@n3, B=@n4.FlapRandomOffset), B=@n4.FlapSpeed))
  Multiply(A=VertexNormalWS(), B=Multiply(A=VertexColor().R, B=Multiply(A=@n5, B=@n4.ZFlappingRange)))
  Multiply(A=@n1.A, B=@n2.A)
  TextureSampleParameter2D() {parameterName=Normal, samplerType=SAMPLERTYPE_Normal, texture=/Game/Effects/Fort_Effects/Textures/Bugs/Bees/T_Bee_Normal}
  VectorParameter() {parameterName=SubSurfColor, defaultValue=<rgba 0.69 0.5743 0.414 1>}
 ]
}
